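# Physics asset for the HealedScar pickaxe mesh, a two-handed melee cosmetic weapon.
# It defines a single body on the Melee_TwoHanded_Handle bone, shaped as one capsule lying along the handle.
# It has no constraints, so the weapon has one simple collision volume and no articulated parts.
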PhysicsAsset /BRCosmetics/Weapons/FORT_Melee/Pickaxe_HealedScar/Meshes/HealedScar_Axe_PhysicsAsset
bodies {Melee_TwoHanded_Handle [{shapeType=Capsule, length=55.9787, radius=4.2396, center=<-0.3794 1.7094 26.6227>, rotation=<rot 1.5331 -90 -0.0052>}]}
constraints []
